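# Physics asset for the F_MED_RoseDepth_Bloom female player body cosmetic, apparently for secondary motion on its left and right leg armor.
# Each side chains three capsule-built dyn_legArmor bones from a small base anchor, looser at the root and nearly rigid below.
# Capsules on pelvis, spine_01 and thighs form the body volume, with bases linked to spine_01 by free angular constraints.
# The sides are asymmetric: thigh_r has tight limits unlike free thigh_l, and armor capsules differ.
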PhysicsAsset /BRCosmetics/Characters/Player/Female/Medium/Bodies/F_MED_RoseDepth_Bloom/Meshes/F_MED_RoseDepth_Bloom_Physics
bodies {
 base_legArmor_l_01 [{shapeType=Capsule, length=0.505, radius=0.505, shapeName=base_legArmor_l_01_capsule, rotation=<rot 0 -90 0>}]
 base_legArmor_r_01 [{shapeType=Capsule, length=0.505, radius=0.505, shapeName=base_legArmor_r_01_capsule, rotation=<rot 0 -90 0>}]
 dyn_legArmor_l_01 [
  {
   shapeType=Capsule
   length=7.625
   radius=2.805
   shapeName=dyn_legArmor_l_01_capsule
   center=<4.7546 0 4.288>
   rotation=<rot 89.8222 -179.977 -179.9772>
  }
  {
   shapeType=Capsule
   length=7.625
   radius=2.805
   shapeName=dyn_legArmor_l_01_capsule
   center=<5.5023 11.4337 -4.0913>
   rotation=<rot 89.8222 -179.977 -179.9772>
  }
 ]
 dyn_legArmor_l_02 [
  {
   shapeType=Capsule
   length=11.6
   radius=2.8
   shapeName=dyn_legArmor_l_02_capsule
   center=<6.7087 0 3.4948>
   rotation=<rot -86.7032 179.9995 -179.9994>
  }
  {
   shapeType=Capsule
   length=11.8544
   radius=3.6686
   shapeName=dyn_legArmor_l_02_capsule
   center=<6.3812 11.602 -3.9251>
   rotation=<rot -83 -170 177>
  }
  {
   shapeType=Capsule
   length=8.0285
   radius=2.7959
   shapeName=dyn_legArmor_l_02_capsule
   center=<4.7192 -10.6472 -2.3056>
   rotation=<rot -81.6 170 -180>
  }
 ]
 dyn_legArmor_l_03 [
  {
   shapeType=Capsule
   length=10.8823
   radius=3.8723
   shapeName=dyn_legArmor_l_03_capsule
   center=<7.4036 -8.1732 0.5445>
   rotation=<rot -4 -7 -122.8715>
  }
  {shapeType=Capsule, length=9.9095, radius=3.6475, shapeName=dyn_legArmor_l_03_capsule, center=<7.3562 8.173 1.0316>, rotation=<rot 0 0 120>}
 ]
 dyn_legArmor_r_01 [{
   shapeType=Capsule
   length=7.625
   radius=2.805
   shapeName=dyn_legArmor_l_01_capsule
   center=<4.7546 0 4.288>
   rotation=<rot 89.8222 -179.977 0.0228>
  }]
 dyn_legArmor_r_02 [
  {shapeType=Capsule, length=11.6, radius=2.8, shapeName=dyn_legArmor_l_02_capsule, center=<6.7087 0 3.4948>, rotation=<rot -88.4 180 0>}
  {
   shapeType=Capsule
   length=11.8544
   radius=3.6686
   shapeName=dyn_legArmor_l_02_capsule
   center=<6.8146 -11.2521 -0.7157>
   rotation=<rot -83 -170 -17>
  }
  {
   shapeType=Capsule
   length=8.0285
   radius=2.7959
   shapeName=dyn_legArmor_l_02_capsule
   center=<4.7192 10.6472 -2.3056>
   rotation=<rot -81.6 -170 0>
  }
 ]
 dyn_legArmor_r_03 [
  {shapeType=Capsule, length=8, radius=3.4, shapeName=dyn_legArmor_l_03_capsule, center=<6.9996 10.1882 -1.6772>, rotation=<rot -4 7 -60>}
  {shapeType=Capsule, length=9.9095, radius=3.6475, shapeName=dyn_legArmor_l_03_capsule, center=<7.3562 -7.8297 0.4371>, rotation=<rot 0 0 60>}
 ]
 pelvis [{shapeType=Capsule, length=8.3135, radius=10.0147, shapeName=pelvis_capsule, center=<-5.0608 -1.9258 0>}]
 spine_01 [{shapeType=Capsule, length=7.7633, radius=8.8967, shapeName=spine_01_capsule}]
 thigh_l [
  {
   shapeType=Capsule
   length=10.5673
   radius=8.2305
   shapeName=thigh_l_capsule
   center=<-10.9328 -0.8081 0.841>
   rotation=<rot 0.1556 87.2523 -85.6174>
  }
  {
   shapeType=Capsule
   length=15.2425
   radius=7.4924
   shapeName=thigh_l_capsule
   center=<-29.3213 -0.2707 1.0969>
   rotation=<rot 0.4219 90.3878 -96.0244>
  }
 ]
 thigh_r [
  {
   shapeType=Capsule
   length=10.5673
   radius=8.2305
   shapeName=thigh_l_capsule
   center=<10.9328 0.8081 -0.841>
   rotation=<rot 0.1556 87.2523 94.3826>
  }
  {
   shapeType=Capsule
   length=15.2425
   radius=7.4924
   shapeName=thigh_l_capsule
   center=<29.3213 0.2707 -1.0969>
   rotation=<rot 0.4219 90.3878 83.9756>
  }
 ]
}
constraints [
 {bone1Name=dyn_legArmor_l_02, bone2Name=dyn_legArmor_l_01, swing1LimitDegrees=3, swing2LimitDegrees=3, twistLimitDegrees=1}
 {bone1Name=dyn_legArmor_l_03, bone2Name=dyn_legArmor_l_02, swing1LimitDegrees=3, swing2LimitDegrees=3, twistLimitDegrees=1}
 {bone1Name=dyn_legArmor_l_01, bone2Name=base_legArmor_l_01, swing1LimitDegrees=25, swing2LimitDegrees=25, twistLimitDegrees=10}
 {
  bone1Name=base_legArmor_l_01
  bone2Name=spine_01
  swing1LimitDegrees=45
  swing1Motion=Free
  swing2LimitDegrees=45
  swing2Motion=Free
  twistLimitDegrees=45
  twistMotion=Free
 }
 {
  bone1Name=base_legArmor_r_01
  bone2Name=spine_01
  swing1LimitDegrees=45
  swing1Motion=Free
  swing2LimitDegrees=45
  swing2Motion=Free
  twistLimitDegrees=45
  twistMotion=Free
 }
 {
  bone1Name=spine_01
  bone2Name=pelvis
  swing1LimitDegrees=45
  swing1Motion=Free
  swing2LimitDegrees=45
  swing2Motion=Free
  twistLimitDegrees=45
  twistMotion=Free
 }
 {
  bone1Name=thigh_l
  bone2Name=pelvis
  swing1LimitDegrees=45
  swing1Motion=Free
  swing2LimitDegrees=45
  swing2Motion=Free
  twistLimitDegrees=45
  twistMotion=Free
 }
 {bone1Name=dyn_legArmor_r_03, bone2Name=dyn_legArmor_r_02, swing1LimitDegrees=3, swing2LimitDegrees=3, twistLimitDegrees=1}
 {bone1Name=thigh_r, bone2Name=pelvis, swing1LimitDegrees=3, swing2LimitDegrees=3, twistLimitDegrees=1}
 {bone1Name=dyn_legArmor_r_02, bone2Name=dyn_legArmor_r_01, swing1LimitDegrees=3, swing2LimitDegrees=3, twistLimitDegrees=1}
 {bone1Name=dyn_legArmor_r_01, bone2Name=base_legArmor_r_01, swing1LimitDegrees=25, swing2LimitDegrees=25, twistLimitDegrees=10}
]
