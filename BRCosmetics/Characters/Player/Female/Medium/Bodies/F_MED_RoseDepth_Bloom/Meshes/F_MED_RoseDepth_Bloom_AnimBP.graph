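# Post-process style AnimBlueprint for the F_MED_RoseDepth_Bloom body cosmetic, adding secondary motion and pose fixes to the copied parent-mesh pose.
# Cached poses chain Control Rig, pre-sim armor offsets, two RigidBody solvers, AnimDynamics on ropes, tassels and calf belt, then action-based corrections.
# The final stage replaces jaw, chin, cheek and ear bones with transforms sampled from HeadMesh.
# The EventGraph updates those transforms, movement flags and smoothed targeting.
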AnimBlueprint /BRCosmetics/Characters/Player/Female/Medium/Bodies/F_MED_RoseDepth_Bloom/Meshes/F_MED_RoseDepth_Bloom_AnimBP
target_skeleton=&/F_MED_RoseDepth_Bloom_Skeleton
graphs [
 AnimGraph {nodes [
   Output Pose(Result=Use cached pose 'transferredJawTransformsCP'()) {layerGroup=None}
   n1 = Control Rig!(
    Source=Copy Pose From Mesh() {bUseAttachedParent=true}
    Alpha
    pub isTargeting float:double=Get isTargeting float() -> (isTargeting float:double)
    pub isUsingHangGlider:bool=Get isUsingHangGliding() -> (isUsingHangGliding:bool)
    pub isPlayingEmote:bool=Get bIsPlayingEmote()
    pub isRBANEnabled:bool=Get bRigidBodyEnabled()
    IsFEIdle:bool=Get isInIdlePoseFrontEnd() -> (isInIdlePoseFrontEnd:bool)
   ) -> (Pose) {
    alphaBoolBlend {blendOption=Linear}
    controlRigAssetReference {blueprintRigClass=&/F_MED_RoseDepth_Bloom_CtrlRig.F_MED_RoseDepth_Bloom_CtrlRig_C}
   }
   Save cached pose 'controlRigCP'(Pose=@n1) {cacheName=controlRigCP}
   n2 = Anim Dynamics - Bone: dyn_neckRope_l_A!(
    ComponentPose=Local To Component(LocalPose=Use cached pose 'rigidBodiesCP'())
    Alpha=1
    GravityOverride:Vector="0,0,-980"
    ExternalForce="0,0,0"
   ) -> (Pose) {
    angularDampingOverride=1
    angularSpringConstant=20000
    bAngularSpring=true
    bOverrideAngularDamping=true
    bUseGravityOverride=true
    gravityScale=0.35
    simulationSpace=RootRelative
    alphaBoolBlend {blendOption=Linear}
    boundBone {boneName=dyn_neckRope_l_A}
    componentAppliedLinearAccClamp=<2500 2500 2500>
    componentLinearAccScale=<1 1 1>
    componentLinearVelScale=<1 1 1>
    gravityOverride {z=-980}
    physicsBodyDefinitions [{
      boundBone {boneName=dyn_neckRope_l_A}
      boxExtents=<20 30 30>
      constraintSetup {angularLimitsMax=<0 15 15>, angularLimitsMin=<0 -15 -15>, angularTarget=<1 0 0>}
      localJointOffset=<4 0 0>
     }]
   }
   n3 = Anim Dynamics - Bone: dyn_neckRope_r_A!(ComponentPose=@n2, Alpha=1, GravityOverride:Vector="0,0,-980", ExternalForce="0,0,0") -> (Pose) {
    angularDampingOverride=1
    angularSpringConstant=20000
    bAngularSpring=true
    bOverrideAngularDamping=true
    bUseGravityOverride=true
    gravityScale=0.35
    simulationSpace=RootRelative
    alphaBoolBlend {blendOption=Linear}
    boundBone {boneName=dyn_neckRope_r_A}
    componentAppliedLinearAccClamp=<2500 2500 2500>
    componentLinearAccScale=<1 1 1>
    componentLinearVelScale=<1 1 1>
    gravityOverride {z=-980}
    physicsBodyDefinitions [{
      boundBone {boneName=dyn_neckRope_r_A}
      boxExtents=<20 30 30>
      constraintSetup {angularLimitsMax=<0 15 15>, angularLimitsMin=<0 -15 -15>, angularTarget=<1 0 0>}
      localJointOffset=<4 0 0>
     }]
   }
   n4 = Anim Dynamics - Bone: dyn_neckRope_l_B!(ComponentPose=@n3, Alpha=1, GravityOverride:Vector="0,0,-980", ExternalForce="0,0,0") -> (Pose) {
    angularDampingOverride=1
    angularSpringConstant=20000
    bAngularSpring=true
    bOverrideAngularDamping=true
    bUseGravityOverride=true
    gravityScale=0.35
    simulationSpace=RootRelative
    alphaBoolBlend {blendOption=Linear}
    boundBone {boneName=dyn_neckRope_l_B}
    componentAppliedLinearAccClamp=<2500 2500 2500>
    componentLinearAccScale=<1 1 1>
    componentLinearVelScale=<1 1 1>
    gravityOverride {z=-980}
    physicsBodyDefinitions [{
      boundBone {boneName=dyn_neckRope_l_B}
      boxExtents=<20 30 30>
      constraintSetup {angularLimitsMax=<0 25 0>, angularLimitsMin=<0 -5 -25>, angularTarget=<1 0 0>}
      localJointOffset=<4 0 0>
     }]
   }
   n5 = Anim Dynamics - Bone: dyn_neckRope_r_B!(ComponentPose=@n4, Alpha=1, GravityOverride:Vector="0,0,-980", ExternalForce="0,0,0") -> (Pose) {
    angularDampingOverride=1
    angularSpringConstant=20000
    bAngularSpring=true
    bOverrideAngularDamping=true
    bUseGravityOverride=true
    gravityScale=0.35
    simulationSpace=RootRelative
    alphaBoolBlend {blendOption=Linear}
    boundBone {boneName=dyn_neckRope_r_B}
    componentAppliedLinearAccClamp=<2500 2500 2500>
    componentLinearAccScale=<1 1 1>
    componentLinearVelScale=<1 1 1>
    gravityOverride {z=-980}
    physicsBodyDefinitions [{
      boundBone {boneName=dyn_neckRope_r_B}
      boxExtents=<20 30 30>
      constraintSetup {angularLimitsMax=<0 25 25>, angularLimitsMin=<0 -5 0>, angularTarget=<1 0 0>}
      localJointOffset=<4 0 0>
     }]
   }
   n6 = Anim Dynamics - Bone: dyn_breastRope_r_01!(ComponentPose=@n5, Alpha=1, GravityOverride:Vector="0,0,-980", ExternalForce="0,0,0") -> (Pose) {
    angularDampingOverride=1
    angularSpringConstant=20000
    bAngularSpring=true
    bOverrideAngularDamping=true
    bUseGravityOverride=true
    gravityScale=0.35
    simulationSpace=RootRelative
    alphaBoolBlend {blendOption=Linear}
    boundBone {boneName=dyn_breastRope_r_01}
    componentAppliedLinearAccClamp=<2500 2500 2500>
    componentLinearAccScale=<1 1 1>
    componentLinearVelScale=<1 1 1>
    gravityOverride {z=-980}
    physicsBodyDefinitions [{
      boundBone {boneName=dyn_breastRope_r_01}
      boxExtents=<20 30 30>
      constraintSetup {angularLimitsMax=<0 5 15>, angularLimitsMin=<0 -15 -15>, angularTarget=<1 0 0>}
      localJointOffset=<4 0 0>
     }]
   }
   n7 = Anim Dynamics - Bone: dyn_breastRope_r_02!(ComponentPose=@n6, Alpha=1, GravityOverride:Vector="0,0,-980", ExternalForce="0,0,0") -> (Pose) {
    angularDampingOverride=1
    angularSpringConstant=20000
    bAngularSpring=true
    bOverrideAngularDamping=true
    bUseGravityOverride=true
    gravityScale=0.35
    simulationSpace=RootRelative
    alphaBoolBlend {blendOption=Linear}
    boundBone {boneName=dyn_breastRope_r_02}
    componentAppliedLinearAccClamp=<2500 2500 2500>
    componentLinearAccScale=<1 1 1>
    componentLinearVelScale=<1 1 1>
    gravityOverride {z=-980}
    physicsBodyDefinitions [{
      boundBone {boneName=dyn_breastRope_r_02}
      boxExtents=<20 30 30>
      constraintSetup {angularLimitsMax=<0 5 15>, angularLimitsMin=<0 -15 -15>, angularTarget=<1 0 0>}
      localJointOffset=<4 0 0>
     }]
   }
   n8 = Anim Dynamics - Bone: dyn_chestRope_r!(ComponentPose=@n7, Alpha=1, GravityOverride:Vector="0,0,-980", ExternalForce="0,0,0") -> (Pose) {
    angularDampingOverride=1
    angularSpringConstant=20000
    bAngularSpring=true
    bOverrideAngularDamping=true
    bUseGravityOverride=true
    gravityScale=0.35
    simulationSpace=RootRelative
    alphaBoolBlend {blendOption=Linear}
    boundBone {boneName=dyn_chestRope_r}
    componentAppliedLinearAccClamp=<2500 2500 2500>
    componentLinearAccScale=<1 1 1>
    componentLinearVelScale=<1 1 1>
    gravityOverride {z=-980}
    physicsBodyDefinitions [{
      boundBone {boneName=dyn_chestRope_r}
      boxExtents=<20 30 30>
      constraintSetup {angularLimitsMax=<0 15 15>, angularLimitsMin=<0 -15 -15>, angularTarget=<1 0 0>}
      localJointOffset=<10 0 0>
     }]
   }
   n9 = Anim Dynamics - Bone: dyn_chestRope_l!(ComponentPose=@n8, Alpha=1, GravityOverride:Vector="0,0,-980", ExternalForce="0,0,0") -> (Pose) {
    angularDampingOverride=1
    angularSpringConstant=20000
    bAngularSpring=true
    bOverrideAngularDamping=true
    bUseGravityOverride=true
    gravityScale=0.35
    simulationSpace=RootRelative
    alphaBoolBlend {blendOption=Linear}
    boundBone {boneName=dyn_chestRope_l}
    componentAppliedLinearAccClamp=<2500 2500 2500>
    componentLinearAccScale=<1 1 1>
    componentLinearVelScale=<1 1 1>
    gravityOverride {z=-980}
    physicsBodyDefinitions [{
      boundBone {boneName=dyn_chestRope_l}
      boxExtents=<20 30 30>
      constraintSetup {angularLimitsMax=<0 15 15>, angularLimitsMin=<0 -15 -15>, angularTarget=<1 0 0>}
      localJointOffset=<10 0 0>
     }]
   }
   n10 = Anim Dynamics - Bone: dyn_tassel_l!(ComponentPose=@n9, Alpha=1, GravityOverride:Vector="0,0,-980", ExternalForce="0,0,0") -> (Pose) {
    angularDampingOverride=1
    angularSpringConstant=20000
    bAngularSpring=true
    bOverrideAngularDamping=true
    bUseGravityOverride=true
    gravityScale=0.35
    simulationSpace=RootRelative
    alphaBoolBlend {blendOption=Linear}
    boundBone {boneName=dyn_tassel_l}
    componentAppliedLinearAccClamp=<2500 2500 2500>
    componentLinearAccScale=<1 1 1>
    componentLinearVelScale=<1 1 1>
    gravityOverride {z=-980}
    physicsBodyDefinitions [{
      boundBone {boneName=dyn_tassel_l}
      boxExtents=<20 30 30>
      constraintSetup {angularLimitsMax=<0 30 30>, angularLimitsMin=<0 -30 -30>, angularTarget=<1 0 0>}
      localJointOffset=<5 0 0>
     }]
   }
   n11 = Anim Dynamics - Bone: dyn_tassel_r!(ComponentPose=@n10, Alpha=1, GravityOverride:Vector="0,0,-980", ExternalForce="0,0,0") -> (Pose) {
    angularDampingOverride=1
    angularSpringConstant=20000
    bAngularSpring=true
    bOverrideAngularDamping=true
    bUseGravityOverride=true
    gravityScale=0.35
    simulationSpace=RootRelative
    alphaBoolBlend {blendOption=Linear}
    boundBone {boneName=dyn_tassel_r}
    componentAppliedLinearAccClamp=<2500 2500 2500>
    componentLinearAccScale=<1 1 1>
    componentLinearVelScale=<1 1 1>
    gravityOverride {z=-980}
    physicsBodyDefinitions [{
      boundBone {boneName=dyn_tassel_r}
      boxExtents=<20 30 30>
      constraintSetup {angularLimitsMax=<0 30 30>, angularLimitsMin=<0 -30 -30>, angularTarget=<1 0 0>}
      localJointOffset=<5 0 0>
     }]
   }
   n12 = Anim Dynamics - Bone: dyn_calfBelt!(ComponentPose=@n11, Alpha=1, GravityOverride:Vector="0,0,-980", ExternalForce="0,0,0") -> (Pose) {
    angularDampingOverride=1
    angularSpringConstant=20000
    bAngularSpring=true
    bOverrideAngularDamping=true
    bUseGravityOverride=true
    gravityScale=0.35
    simulationSpace=RootRelative
    alphaBoolBlend {blendOption=Linear}
    boundBone {boneName=dyn_calfBelt}
    componentAppliedLinearAccClamp=<2500 2500 2500>
    componentLinearAccScale=<1 1 1>
    componentLinearVelScale=<1 1 1>
    gravityOverride {z=-980}
    physicsBodyDefinitions [{
      boundBone {boneName=dyn_calfBelt}
      boxExtents=<20 30 30>
      constraintSetup {angularLimitsMax=<0 30 50>, angularLimitsMin=<0 -30 -15>, angularTarget=<1 0 0>}
      localJointOffset=<8 0 0>
     }]
   }
   Save cached pose 'animDynCP'(Pose=Component To Local(ComponentPose=@n12)) {cacheName=animDynCP}
   n13 = Get GravityOverride()
   n14 = RigidBody!(ComponentPose=Local To Component(LocalPose=Use cached pose 'presimFixesCP'()), Alpha, OverrideWorldGravity:Vector=@n13, ExternalForce) -> (Pose) {
    bClampLinearTranslationLimitToRefPose=true
    bOverrideWorldGravity=true
    overridePhysicsAsset=&/F_MED_RoseDepth_Bloom_Physics
    simulationSpace=BaseBoneSpace
    alphaBoolBlend {blendOption=Linear}
    baseBoneRef {boneName=pelvis}
    componentLinearAccScale=<0.4 0.4 0.4>
    componentLinearVelScale=<0.4 0.4 0.4>
    simSpaceSettings {worldAlpha=0.2}
   }
   n15 = RigidBody(ComponentPose=@n14, OverrideWorldGravity=@n13) {
    bClampLinearTranslationLimitToRefPose=true
    bOverrideWorldGravity=true
    overridePhysicsAsset=&/F_MED_RoseDepth_Bloom_Physics_backFront
    simulationSpace=BaseBoneSpace
    alphaBoolBlend {blendOption=Linear}
    baseBoneRef {boneName=pelvis}
    componentLinearAccScale=<0.4 0.4 0.4>
    componentLinearVelScale=<0.4 0.4 0.4>
    simSpaceSettings {worldAlpha=0.2}
   }
   Save cached pose 'rigidBodiesCP'(Pose=Component To Local(ComponentPose=@n15)) {cacheName=rigidBodiesCP}
   n16 = Get bIsSkydiveFloating()
   n17 = "Transform (Modify) Bone - Bone: pvtBase_shoulderArmor_l"!(
    ComponentPose=Local To Component(LocalPose=Use cached pose 'animDynCP'())
    bAlphaBoolEnabled:bool=@n16
    Translation="0,0,0"
    Rotation="0,0,-50"
   ) -> (Pose) {
    alphaInputType=Bool
    rotationMode=BMM_Additive
    rotationSpace=BCS_BoneSpace
    translationMode=BMM_Additive
    translationSpace=BCS_BoneSpace
    alphaBoolBlend {blendInTime=1, blendOption=Linear, blendOutTime=2}
    boneToModify {boneName=pvtBase_shoulderArmor_l}
    rotation {roll=-50}
   }
   n18 = "Transform (Modify) Bone - Bone: base_backArmor_01"!(ComponentPose=@n17, bAlphaBoolEnabled:bool=@n16, Translation="0,0,1", Rotation="10,0,0") -> (Pose) {
    alphaInputType=Bool
    rotationMode=BMM_Additive
    rotationSpace=BCS_BoneSpace
    translationMode=BMM_Additive
    translationSpace=BCS_BoneSpace
    alphaBoolBlend {blendInTime=1, blendOption=Cubic, blendOutTime=1}
    boneToModify {boneName=base_backArmor_01}
    rotation {pitch=10}
    translation {z=1}
   }
   n19 = "Transform (Modify) Bone - Bone: pvtBase_shoulderArmor_l"(ComponentPose=@n18, bAlphaBoolEnabled=Get isUsingHangGliding(), Translation="0,0,0", Rotation="0,0,-10") {
    alphaInputType=Bool
    rotationMode=BMM_Additive
    rotationSpace=BCS_BoneSpace
    translationMode=BMM_Additive
    translationSpace=BCS_BoneSpace
    alphaBoolBlend {blendInTime=2, blendOption=Linear, blendOutTime=2}
    boneToModify {boneName=pvtBase_shoulderArmor_l}
    rotation {roll=-10}
   }
   n20 = "Transform (Modify) Bone - Bone: attach_cape"!(ComponentPose=@n19, bAlphaBoolEnabled:bool=Get bIsSkydiveDiveMode(), Translation="0,0,0", Rotation="0,0,20") -> (Pose) {
    alphaInputType=Bool
    rotationMode=BMM_Additive
    rotationSpace=BCS_BoneSpace
    translationMode=BMM_Additive
    translationSpace=BCS_BoneSpace
    alphaBoolBlend {blendInTime=2, blendOption=Sinusoidal, blendOutTime=2}
    boneToModify {boneName=attach_cape}
    rotation {roll=20}
   }
   Save cached pose 'actionFixesCP'(Pose=Component To Local(ComponentPose=@n20)) {cacheName=actionFixesCP}
   n21 = "Transform (Modify) Bone - Bone: FACIAL_C_Chin"!(
    ComponentPose=Local To Component(LocalPose=Use cached pose 'postSimFixes'())
    bAlphaBoolEnabled:bool=True
    Translation=Get FACIAL_C_Chin_Translation() -> (FACIAL_C_Chin_Translation:Vector)
    Rotation=Get FACIAL_C_Chin_Rotation() -> (FACIAL_C_Chin_Rotation:Rotator)
    Scale="1,1,1"
   ) -> (Pose) {
    alphaInputType=Bool
    rotationMode=BMM_Replace
    rotationSpace=BCS_ParentBoneSpace
    translationMode=BMM_Replace
    translationSpace=BCS_ParentBoneSpace
    alphaBoolBlend {blendOption=Linear}
    boneToModify {boneName=FACIAL_C_Chin}
   }
   n22 = "Transform (Modify) Bone - Bone: FACIAL_L_UnderChin"!(
    ComponentPose=@n21
    bAlphaBoolEnabled:bool=True
    Translation=Get FACIAL_L_UnderChin_Translation() -> (FACIAL_L_UnderChin_Translation:Vector)
    Rotation=Get FACIAL_L_UnderChin_Rotation() -> (FACIAL_L_UnderChin_Rotation:Rotator)
    Scale="1,1,1"
   ) -> (Pose) {
    alphaInputType=Bool
    rotationMode=BMM_Replace
    rotationSpace=BCS_ParentBoneSpace
    translationMode=BMM_Replace
    translationSpace=BCS_ParentBoneSpace
    alphaBoolBlend {blendOption=Linear}
    boneToModify {boneName=FACIAL_L_UnderChin}
   }
   n23 = "Transform (Modify) Bone - Bone: FACIAL_R_UnderChin"!(
    ComponentPose=@n22
    bAlphaBoolEnabled:bool=True
    Translation=Get FACIAL_R_UnderChin_Translation() -> (FACIAL_R_UnderChin_Translation:Vector)
    Rotation=Get FACIAL_R_UnderChin_Rotation() -> (FACIAL_R_UnderChin_Rotation:Rotator)
    Scale="1,1,1"
   ) -> (Pose) {
    alphaInputType=Bool
    rotationMode=BMM_Replace
    rotationSpace=BCS_ParentBoneSpace
    translationMode=BMM_Replace
    translationSpace=BCS_ParentBoneSpace
    alphaBoolBlend {blendOption=Linear}
    boneToModify {boneName=FACIAL_R_UnderChin}
   }
   n24 = "Transform (Modify) Bone - Bone: FACIAL_L_Jawline"!(
    ComponentPose=@n23
    bAlphaBoolEnabled:bool=True
    Translation=Get FACIAL_L_Jawline_Translation() -> (FACIAL_L_Jawline_Translation:Vector)
    Rotation=Get FACIAL_L_Jawline_Rotation() -> (FACIAL_L_Jawline_Rotation:Rotator)
    Scale="1,1,1"
   ) -> (Pose) {
    alphaInputType=Bool
    rotationMode=BMM_Replace
    rotationSpace=BCS_ParentBoneSpace
    translationMode=BMM_Replace
    translationSpace=BCS_ParentBoneSpace
    alphaBoolBlend {blendOption=Linear}
    boneToModify {boneName=FACIAL_L_Jawline}
   }
   n25 = "Transform (Modify) Bone - Bone: FACIAL_R_Jawline"!(
    ComponentPose=@n24
    bAlphaBoolEnabled:bool=True
    Translation=Get FACIAL_R_Jawline_Translation() -> (FACIAL_R_Jawline_Translation:Vector)
    Rotation=Get FACIAL_R_Jawline_Rotation() -> (FACIAL_R_Jawline_Rotation:Rotator)
    Scale="1,1,1"
   ) -> (Pose) {
    alphaInputType=Bool
    rotationMode=BMM_Replace
    rotationSpace=BCS_ParentBoneSpace
    translationMode=BMM_Replace
    translationSpace=BCS_ParentBoneSpace
    alphaBoolBlend {blendOption=Linear}
    boneToModify {boneName=FACIAL_R_Jawline}
   }
   n26 = "Transform (Modify) Bone - Bone: FACIAL_C_Jaw"!(
    ComponentPose=@n25
    bAlphaBoolEnabled:bool=True
    Translation=Get FACIAL_C_Jaw_Tranform() -> (FACIAL_C_Jaw_Tranform:Vector)
    Rotation=Get FACIAL_C_Jaw_Rotation() -> (FACIAL_C_Jaw_Rotation:Rotator)
    Scale="1,1,1"
   ) -> (Pose) {
    alphaInputType=Bool
    rotationMode=BMM_Replace
    rotationSpace=BCS_ParentBoneSpace
    translationMode=BMM_Replace
    translationSpace=BCS_ParentBoneSpace
    alphaBoolBlend {blendOption=Linear}
    boneToModify {boneName=FACIAL_C_Jaw}
   }
   n27 = "Transform (Modify) Bone - Bone: FACIAL_C_Jawline"!(
    ComponentPose=@n26
    bAlphaBoolEnabled:bool=True
    Translation=Get FACIAL_C_Jawline_Translation() -> (FACIAL_C_Jawline_Translation:Vector)
    Rotation=Get FACIAL_C_Jawline_Rotation() -> (FACIAL_C_Jawline_Rotation:Rotator)
    Scale="1,1,1"
   ) -> (Pose) {
    alphaInputType=Bool
    rotationMode=BMM_Replace
    rotationSpace=BCS_ParentBoneSpace
    translationMode=BMM_Replace
    translationSpace=BCS_ParentBoneSpace
    alphaBoolBlend {blendOption=Linear}
    boneToModify {boneName=FACIAL_C_Jawline}
   }
   n28 = "Transform (Modify) Bone - Bone: FACIAL_L_ChinSide"!(
    ComponentPose=@n27
    bAlphaBoolEnabled:bool=True
    Translation=Get FACIAL_L_ChinSide_Translation() -> (FACIAL_L_ChinSide_Translation:Vector)
    Rotation=Get FACIAL_L_ChinSide_Rotation() -> (FACIAL_L_ChinSide_Rotation:Rotator)
    Scale="1,1,1"
   ) -> (Pose) {
    alphaInputType=Bool
    rotationMode=BMM_Replace
    rotationSpace=BCS_ParentBoneSpace
    translationMode=BMM_Replace
    translationSpace=BCS_ParentBoneSpace
    alphaBoolBlend {blendOption=Linear}
    boneToModify {boneName=FACIAL_L_ChinSide}
   }
   n29 = "Transform (Modify) Bone - Bone: FACIAL_R_ChinSide"!(
    ComponentPose=@n28
    bAlphaBoolEnabled:bool=True
    Translation=Get FACIAL_R_ChinSide_Translation() -> (FACIAL_R_ChinSide_Translation:Vector)
    Rotation=Get FACIAL_R_ChinSide_Rotation() -> (FACIAL_R_ChinSide_Rotation:Rotator)
    Scale="1,1,1"
   ) -> (Pose) {
    alphaInputType=Bool
    rotationMode=BMM_Replace
    rotationSpace=BCS_ParentBoneSpace
    translationMode=BMM_Replace
    translationSpace=BCS_ParentBoneSpace
    alphaBoolBlend {blendOption=Linear}
    boneToModify {boneName=FACIAL_R_ChinSide}
   }
   n30 = "Transform (Modify) Bone - Bone: FACIAL_L_Ear"!(
    ComponentPose=@n29
    bAlphaBoolEnabled:bool=True
    Translation=Get FACIAL_L_Ear_Translation() -> (FACIAL_L_Ear_Translation:Vector)
    Rotation=Get FACIAL_L_Ear_Rotation() -> (FACIAL_L_Ear_Rotation:Rotator)
    Scale="1,1,1"
   ) -> (Pose) {
    alphaInputType=Bool
    rotationMode=BMM_Replace
    rotationSpace=BCS_ParentBoneSpace
    translationMode=BMM_Replace
    translationSpace=BCS_ParentBoneSpace
    alphaBoolBlend {blendOption=Linear}
    boneToModify {boneName=FACIAL_L_Ear}
   }
   n31 = "Transform (Modify) Bone - Bone: FACIAL_R_Ear"!(
    ComponentPose=@n30
    bAlphaBoolEnabled:bool=True
    Translation=Get FACIAL_R_Ear_Translation() -> (FACIAL_R_Ear_Translation:Vector)
    Rotation=Get FACIAL_R_Ear_Rotation() -> (FACIAL_R_Ear_Rotation:Rotator)
    Scale="1,1,1"
   ) -> (Pose) {
    alphaInputType=Bool
    rotationMode=BMM_Replace
    rotationSpace=BCS_ParentBoneSpace
    translationMode=BMM_Replace
    translationSpace=BCS_ParentBoneSpace
    alphaBoolBlend {blendOption=Linear}
    boneToModify {boneName=FACIAL_R_Ear}
   }
   n32 = "Transform (Modify) Bone - Bone: FACIAL_L_CheekLower"!(
    ComponentPose=@n31
    bAlphaBoolEnabled:bool=True
    Translation=Get FACIAL_L_CheekLower_Translation() -> (FACIAL_L_CheekLower_Translation:Vector)
    Rotation=Get FACIAL_L_CheekLower_Rotation() -> (FACIAL_L_CheekLower_Rotation:Rotator)
    Scale="1,1,1"
   ) -> (Pose) {
    alphaInputType=Bool
    rotationMode=BMM_Replace
    rotationSpace=BCS_ParentBoneSpace
    translationMode=BMM_Replace
    translationSpace=BCS_ParentBoneSpace
    alphaBoolBlend {blendOption=Linear}
    boneToModify {boneName=FACIAL_L_CheekLower}
   }
   n33 = "Transform (Modify) Bone - Bone: FACIAL_R_CheekLower"!(
    ComponentPose=@n32
    bAlphaBoolEnabled:bool=True
    Translation=Get FACIAL_R_CheekLower_Translation() -> (FACIAL_R_CheekLower_Translation:Vector)
    Rotation=Get FACIAL_R_CheekLower_Rotation() -> (FACIAL_R_CheekLower_Rotation:Rotator)
    Scale="1,1,1"
   ) -> (Pose) {
    alphaInputType=Bool
    rotationMode=BMM_Replace
    rotationSpace=BCS_ParentBoneSpace
    translationMode=BMM_Replace
    translationSpace=BCS_ParentBoneSpace
    alphaBoolBlend {blendOption=Linear}
    boneToModify {boneName=FACIAL_R_CheekLower}
   }
   Save cached pose 'transferredJawTransformsCP'(Pose=Component To Local(ComponentPose=@n33)) {cacheName=transferredJawTransformsCP}
   n34 = Get onZipline() -> (onZipline:bool)
   n35 = "Transform (Modify) Bone - Bone: base_frontArmor_01"!(
    ComponentPose=Local To Component(LocalPose=Use cached pose 'controlRigCP'())
    bAlphaBoolEnabled:bool=Get bIsDBNO()
    Translation="0,0,0"
    Rotation="10,0,0"
   ) -> (Pose) {
    alphaInputType=Bool
    rotationMode=BMM_Additive
    rotationSpace=BCS_BoneSpace
    translationMode=BMM_Additive
    translationSpace=BCS_BoneSpace
    alphaBoolBlend {blendOption=Linear}
    boneToModify {boneName=base_frontArmor_01}
    rotation {pitch=10}
   }
   n36 = "Transform (Modify) Bone - Bone: base_legArmor_r_01"!(ComponentPose=@n35, bAlphaBoolEnabled:bool=True, Translation="0,0,0", Rotation="15,0,-10", Scale="1,1,1") -> (Pose) {
    alphaInputType=Bool
    rotationMode=BMM_Additive
    rotationSpace=BCS_BoneSpace
    translationMode=BMM_Additive
    translationSpace=BCS_BoneSpace
    alphaBoolBlend {blendInTime=0.1, blendOption=Cubic, blendOutTime=0.1}
    boneToModify {boneName=base_legArmor_r_01}
    rotation {pitch=15, roll=-10}
   }
   n37 = "Transform (Modify) Bone - Bone: base_legArmor_l_01"!(ComponentPose=@n36, bAlphaBoolEnabled:bool=True, Translation="0,0,0", Rotation="15,0,10", Scale="1,1,1") -> (Pose) {
    alphaInputType=Bool
    rotationMode=BMM_Additive
    rotationSpace=BCS_BoneSpace
    translationMode=BMM_Additive
    translationSpace=BCS_BoneSpace
    alphaBoolBlend {blendInTime=0.2, blendOption=Cubic, blendOutTime=0.2}
    boneToModify {boneName=base_legArmor_l_01}
    rotation {pitch=15, roll=10}
   }
   n38 = "Transform (Modify) Bone - Bone: base_frontArmor_01"!(ComponentPose=@n37, bAlphaBoolEnabled=True, Translation="0,0,0", Rotation="-20,0,0", Scale:Vector="1,1,1") -> (Pose) {
    alphaInputType=Bool
    rotationMode=BMM_Additive
    rotationSpace=BCS_BoneSpace
    translationMode=BMM_Additive
    translationSpace=BCS_BoneSpace
    alphaBoolBlend {blendInTime=0.2, blendOption=Cubic, blendOutTime=0.2}
    boneToModify {boneName=base_frontArmor_01}
    rotation {pitch=-20}
   }
   n39 = "Transform (Modify) Bone - Bone: base_backArmor_01"!(ComponentPose=@n38, bAlphaBoolEnabled=True, Translation="0,0,0", Rotation="30,0,0", Scale:Vector="1,1,1") -> (Pose) {
    alphaInputType=Bool
    rotationMode=BMM_Additive
    rotationSpace=BCS_BoneSpace
    translationMode=BMM_Additive
    translationSpace=BCS_BoneSpace
    alphaBoolBlend {blendInTime=0.2, blendOption=Cubic, blendOutTime=0.2}
    boneToModify {boneName=base_backArmor_01}
    rotation {pitch=30}
   }
   n40 = "Transform (Modify) Bone - Bone: base_legArmor_r_01"(ComponentPose=@n39, bAlphaBoolEnabled=True, Translation="0,0,0", Rotation="60,0,0", Scale="1,1,1") {
    alphaInputType=Bool
    rotationMode=BMM_Additive
    rotationSpace=BCS_BoneSpace
    translationMode=BMM_Additive
    translationSpace=BCS_BoneSpace
    alphaBoolBlend {blendInTime=0.2, blendOption=Cubic, blendOutTime=0.2}
    boneToModify {boneName=base_legArmor_r_01}
    rotation {pitch=60}
   }
   n41 = "Transform (Modify) Bone - Bone: base_frontArmor_01"(ComponentPose=@n40, bAlphaBoolEnabled=True, Translation="0,0,2", Rotation="0,0,0", Scale="1,1,1") {
    alphaInputType=Bool
    rotationMode=BMM_Additive
    rotationSpace=BCS_BoneSpace
    translationMode=BMM_Additive
    translationSpace=BCS_BoneSpace
    alphaBoolBlend {blendInTime=0.2, blendOption=Cubic, blendOutTime=0.2}
    boneToModify {boneName=base_frontArmor_01}
    translation {z=2}
   }
   n42 = "Transform (Modify) Bone - Bone: base_frontArmor_01"(ComponentPose=@n41, bAlphaBoolEnabled=True, Translation="0,0,2", Rotation="5,0,0", Scale="1,1,1") {
    alphaInputType=Bool
    rotationMode=BMM_Additive
    rotationSpace=BCS_BoneSpace
    translationMode=BMM_Additive
    translationSpace=BCS_BoneSpace
    alphaBoolBlend {blendInTime=0.2, blendOption=Cubic, blendOutTime=0.2}
    boneToModify {boneName=base_frontArmor_01}
    rotation {pitch=5}
    translation {z=2}
   }
   n43 = "Transform (Modify) Bone - Bone: base_frontArmor_01"(ComponentPose=@n42, bAlphaBoolEnabled=True, Translation="0,0,0", Rotation="5,0,0", Scale="1,1,1") {
    alphaInputType=Bool
    rotationMode=BMM_Additive
    rotationSpace=BCS_BoneSpace
    translationMode=BMM_Additive
    translationSpace=BCS_BoneSpace
    alphaBoolBlend {blendInTime=0.2, blendOption=Cubic, blendOutTime=0.2}
    boneToModify {boneName=base_frontArmor_01}
    rotation {pitch=5}
   }
   n44 = "Transform (Modify) Bone - Bone: base_frontArmor_01"(
    ComponentPose=@n43
    bAlphaBoolEnabled=Get isCrouchingStill() -> (isCrouchingStill:bool)
    Translation="0,0,0"
    Rotation="0,0,0"
    Scale="1,1,1"
   ) {
    alphaInputType=Bool
    rotationMode=BMM_Additive
    rotationSpace=BCS_BoneSpace
    translationMode=BMM_Additive
    translationSpace=BCS_BoneSpace
    alphaBoolBlend {blendInTime=0.2, blendOption=Cubic, blendOutTime=0.2}
    boneToModify {boneName=base_frontArmor_01}
   }
   n45 = Get isStandingStill() -> (isStandingStill:bool)
   n46 = "Transform (Modify) Bone - Bone: base_legArmor_l_01"(ComponentPose=@n44, bAlphaBoolEnabled=@n45, Translation="0,0,0", Rotation="0,0,0", Scale="1,1,1") {
    alphaInputType=Bool
    rotationMode=BMM_Additive
    rotationSpace=BCS_BoneSpace
    translationMode=BMM_Additive
    translationSpace=BCS_BoneSpace
    alphaBoolBlend {blendInTime=0.1, blendOption=Cubic, blendOutTime=0.2}
    boneToModify {boneName=base_legArmor_l_01}
   }
   n47 = "Transform (Modify) Bone - Bone: dyn_legArmor_r_01"!(ComponentPose=@n46, bAlphaBoolEnabled:bool=@n45, Translation="0,0,0", Rotation="0,0,0", Scale="1,1,1") -> (Pose) {
    alphaInputType=Bool
    rotationMode=BMM_Additive
    rotationSpace=BCS_BoneSpace
    translationMode=BMM_Additive
    translationSpace=BCS_BoneSpace
    alphaBoolBlend {blendInTime=0.1, blendOption=Cubic, blendOutTime=0.2}
    boneToModify {boneName=dyn_legArmor_r_01}
   }
   n48 = "Transform (Modify) Bone - Bone: base_legArmor_bendB_l_01"!(ComponentPose=@n47, bAlphaBoolEnabled:bool=@n45, Translation="0,0,1", Rotation="0,0,0", Scale="1,1,1") -> (Pose) {
    alphaInputType=Bool
    rotationMode=BMM_Additive
    rotationSpace=BCS_BoneSpace
    translationMode=BMM_Additive
    translationSpace=BCS_BoneSpace
    alphaBoolBlend {blendInTime=0.1, blendOption=Cubic, blendOutTime=0.2}
    boneToModify {boneName=base_legArmor_bendB_l_01}
    translation {z=1}
   }
   n49 = "Transform (Modify) Bone - Bone: base_legArmor_l_01"(ComponentPose=@n48, bAlphaBoolEnabled=@n34, Translation="0,0,0", Rotation="0,0,0", Scale="1,1,1") {
    alphaInputType=Bool
    rotationMode=BMM_Additive
    rotationSpace=BCS_BoneSpace
    translationMode=BMM_Additive
    translationSpace=BCS_BoneSpace
    alphaBoolBlend {blendInTime=0.2, blendOption=Cubic, blendOutTime=0.2}
    boneToModify {boneName=base_legArmor_l_01}
   }
   n50 = "Transform (Modify) Bone - Bone: dyn_legArmor_bendF_l_01"!(ComponentPose=@n49, bAlphaBoolEnabled:bool=@n34, Translation="-1,-1,5", Rotation="0,0,0", Scale="1,1,1") -> (Pose) {
    alphaInputType=Bool
    rotationMode=BMM_Additive
    rotationSpace=BCS_BoneSpace
    translationMode=BMM_Additive
    translationSpace=BCS_BoneSpace
    alphaBoolBlend {blendInTime=0.2, blendOption=Cubic, blendOutTime=0.2}
    boneToModify {boneName=dyn_legArmor_bendF_l_01}
    translation=<-1 -1 5>
   }
   n51 = "Transform (Modify) Bone - Bone: base_legArmor_l_01"(ComponentPose=@n50, bAlphaBoolEnabled=Get bIsFalling(), Translation="0,0,0", Rotation="0,0,0", Scale="1,1,1") {
    alphaInputType=Bool
    rotationMode=BMM_Additive
    rotationSpace=BCS_BoneSpace
    translationMode=BMM_Additive
    translationSpace=BCS_BoneSpace
    alphaBoolBlend {blendInTime=0.2, blendOption=Cubic, blendOutTime=0.2}
    boneToModify {boneName=base_legArmor_l_01}
   }
   Save cached pose 'presimFixesCP'(Pose=Component To Local(ComponentPose=@n51)) {cacheName=presimFixesCP}
   n52 = Get bIsTacticalSprinting()
   n53 = "Transform (Modify) Bone - Bone: dyn_legArmor_l_01"!(
    ComponentPose=Local To Component(LocalPose=Use cached pose 'actionFixesCP'())
    bAlphaBoolEnabled:bool=@n52
    Translation="0,0,0"
    Rotation="-10,0,0"
   ) -> (Pose) {
    alphaInputType=Bool
    rotationMode=BMM_Additive
    rotationSpace=BCS_BoneSpace
    translationMode=BMM_Additive
    translationSpace=BCS_BoneSpace
    alphaBoolBlend {blendInTime=2, blendOption=Linear, blendOutTime=2}
    boneToModify {boneName=dyn_legArmor_l_01}
    rotation {pitch=-10}
   }
   n54 = "Transform (Modify) Bone - Bone: dyn_legArmor_r_01"!(ComponentPose=@n53, bAlphaBoolEnabled=@n52, Translation="0,0,0", Rotation="-10,0,0") -> (Pose) {
    alphaInputType=Bool
    rotationMode=BMM_Additive
    rotationSpace=BCS_BoneSpace
    translationMode=BMM_Additive
    translationSpace=BCS_BoneSpace
    alphaBoolBlend {blendInTime=2, blendOption=Linear, blendOutTime=2}
    boneToModify {boneName=dyn_legArmor_r_01}
    rotation {pitch=-10}
   }
   n55 = "Transform (Modify) Bone - Bone: dyn_legArmor_r_01"!(ComponentPose=@n54, bAlphaBoolEnabled=True, Translation="0,0,0", Rotation="30,0,-10", Scale:Vector="1,1,1") -> (Pose) {
    alphaInputType=Bool
    rotationMode=BMM_Additive
    rotationSpace=BCS_BoneSpace
    translationMode=BMM_Additive
    translationSpace=BCS_BoneSpace
    alphaBoolBlend {blendInTime=0.2, blendOption=Cubic, blendOutTime=0.2}
    boneToModify {boneName=dyn_legArmor_r_01}
    rotation {pitch=30, roll=-10}
   }
   n56 = "Transform (Modify) Bone - Bone: dyn_backArmor_01"!(ComponentPose=@n55, bAlphaBoolEnabled:bool=True, Translation="0,0,0", Rotation="40,0,30", Scale="1,1,1") -> (Pose) {
    alphaInputType=Bool
    rotationMode=BMM_Additive
    rotationSpace=BCS_BoneSpace
    translationMode=BMM_Additive
    translationSpace=BCS_BoneSpace
    alphaBoolBlend {blendInTime=0.2, blendOption=Cubic, blendOutTime=0.2}
    boneToModify {boneName=dyn_backArmor_01}
    rotation {pitch=40, roll=30}
   }
   n57 = "Transform (Modify) Bone - Bone: dyn_backArmor_02"!(ComponentPose=@n56, bAlphaBoolEnabled:bool=True, Translation="0,0,0", Rotation="40,0,0", Scale="1,1,1") -> (Pose) {
    alphaInputType=Bool
    rotationMode=BMM_Additive
    rotationSpace=BCS_BoneSpace
    translationMode=BMM_Additive
    translationSpace=BCS_BoneSpace
    alphaBoolBlend {blendInTime=0.2, blendOption=Cubic, blendOutTime=0.2}
    boneToModify {boneName=dyn_backArmor_02}
    rotation {pitch=40}
   }
   n58 = "Transform (Modify) Bone - Bone: dyn_legArmor_l_01"!(ComponentPose=@n57, bAlphaBoolEnabled=Get isCrouchingStill(), Translation="0,0,0", Rotation="-15,0,0", Scale:Vector="1,1,1") -> (Pose) {
    alphaInputType=Bool
    rotationMode=BMM_Additive
    rotationSpace=BCS_BoneSpace
    translationMode=BMM_Additive
    translationSpace=BCS_BoneSpace
    alphaBoolBlend {blendInTime=0.5, blendOption=Cubic, blendOutTime=0.5}
    boneToModify {boneName=dyn_legArmor_l_01}
    rotation {pitch=-15}
   }
   n59 = Get bIsCrouchSprinting()
   n60 = "Transform (Modify) Bone - Bone: base_legArmor_bendF_l_01"!(ComponentPose=@n58, bAlphaBoolEnabled:bool=@n59, Translation="0,0,3", Rotation="0,0,0", Scale="1,1,1") -> (Pose) {
    alphaInputType=Bool
    rotationMode=BMM_Additive
    rotationSpace=BCS_BoneSpace
    translationMode=BMM_Additive
    translationSpace=BCS_BoneSpace
    alphaBoolBlend {blendInTime=0.5, blendOption=Cubic, blendOutTime=0.5}
    boneToModify {boneName=base_legArmor_bendF_l_01}
    translation {z=3}
   }
   n61 = "Transform (Modify) Bone - Bone: dyn_legArmor_l_02"!(ComponentPose=@n60, bAlphaBoolEnabled:bool=@n59, Translation="0,0,0", Rotation="-10,0,0", Scale="1,1,1") -> (Pose) {
    alphaInputType=Bool
    rotationMode=BMM_Additive
    rotationSpace=BCS_BoneSpace
    translationMode=BMM_Additive
    translationSpace=BCS_BoneSpace
    alphaBoolBlend {blendInTime=0.5, blendOption=Cubic, blendOutTime=0.5}
    boneToModify {boneName=dyn_legArmor_l_02}
    rotation {pitch=-10}
   }
   n62 = "Transform (Modify) Bone - Bone: dyn_legArmor_l_03"!(ComponentPose=@n61, bAlphaBoolEnabled:bool=@n59, Translation="0,0,0", Rotation="-5,0,0", Scale="1,1,1") -> (Pose) {
    alphaInputType=Bool
    rotationMode=BMM_Additive
    rotationSpace=BCS_BoneSpace
    translationMode=BMM_Additive
    translationSpace=BCS_BoneSpace
    alphaBoolBlend {blendInTime=0.5, blendOption=Cubic, blendOutTime=0.5}
    boneToModify {boneName=dyn_legArmor_l_03}
    rotation {pitch=-5}
   }
   Save cached pose 'postSimFixes'(Pose=Component To Local(ComponentPose=@n62)) {cacheName=postSimFixes}
   Reroute Node(InputPin:Wildcard) -> (OutputPin:Wildcard)
  ]}
 EventGraph {nodes [
   n1 = Event BlueprintUpdateAnimation()
   > Set Delta Time X(Delta Time X=@n1.DeltaTimeX)
   > n2 = Sequence!(execute) -> (then_0, then_1, then_2:Exec, then_3:Exec, then_4:Exec, then_5:Exec)
   n3 = Get bIsTargeting()
   Set isTargeting float(
    execute=@n2.then_1
    isTargeting float:double=FInterpTo(Current=Get isTargeting float(), Target="To Float (Boolean)"(InBool=@n3), DeltaTime=Get Delta Time X(), InterpSpeed=15)
   ) -> (Output_Get:double)
   n4 = Get bIsCrouching()
   n5 = Get bIsMoving2D()
   Set isStandingStill(
    execute=@n2.then_3
    isStandingStill:bool=AND Boolean!(A=Get bIsOnGround() -> (bIsOnGround:bool), B=NOT Boolean(A=@n4), C:bool=NOT Boolean(A=@n5)) -> (ReturnValue)
   ) -> (Output_Get:bool)
   n6 = Get HeadMesh() -> (HeadMesh:Skeletal Mesh Component Budgeted Object Reference)
   Is Valid(exec=@n2.then_5, InputObject=@n6)
   n7 = GetSocketBoneName!(self:Skinned Mesh Component Object Reference=@n6, InSocketName:Name=FACIAL_C_Jaw) -> (ReturnValue:Name=None)
   > n8 = Set isUsingFacialBones(isUsingFacialBones:bool="Not Equal Exactly (String)"(A="To String (Name)"(InName=@n7), B=None)) -> (Output_Get:bool)
   > Branch(Condition=@n8)
   n9 = Get HeadMesh()
   n10 = BreakTransform(InTransform=GetSocketTransform(self=@n9, InSocketName=FACIAL_C_Chin, TransformSpace=RTS_ParentBoneSpace))
   > Set FACIAL_C_Chin_Translation(FACIAL_C_Chin_Translation:Vector=@n10) -> (Output_Get:Vector)
   > Set FACIAL_C_Chin_Rotation(FACIAL_C_Chin_Rotation:Rotator=@n10.Rotation) -> (Output_Get:Rotator)
   n11 = BreakTransform(InTransform=GetSocketTransform(self=@n9, InSocketName=FACIAL_L_UnderChin, TransformSpace=RTS_ParentBoneSpace))
   > Set FACIAL_L_UnderChin_Translation(FACIAL_L_UnderChin_Translation:Vector=@n11) -> (Output_Get:Vector)
   > Set FACIAL_L_UnderChin_Rotation(FACIAL_L_UnderChin_Rotation:Rotator=@n11.Rotation) -> (Output_Get:Rotator)
   n12 = BreakTransform(InTransform=GetSocketTransform(self=@n9, InSocketName=FACIAL_R_UnderChin, TransformSpace=RTS_ParentBoneSpace))
   > Set FACIAL_R_UnderChin_Translation(FACIAL_R_UnderChin_Translation:Vector=@n12) -> (Output_Get:Vector)
   > Set FACIAL_R_UnderChin_Rotation(FACIAL_R_UnderChin_Rotation:Rotator=@n12.Rotation) -> (Output_Get:Rotator)
   n13 = BreakTransform(InTransform=GetSocketTransform(self=@n9, InSocketName=FACIAL_C_Jaw, TransformSpace=RTS_ParentBoneSpace))
   > Set FACIAL_C_Jaw_Tranform(FACIAL_C_Jaw_Tranform:Vector=@n13) -> (Output_Get:Vector)
   > Set FACIAL_C_Jaw_Rotation(FACIAL_C_Jaw_Rotation:Rotator=@n13.Rotation) -> (Output_Get:Rotator)
   n14 = BreakTransform(InTransform=GetSocketTransform(self=@n9, InSocketName=FACIAL_C_Jawline, TransformSpace=RTS_ParentBoneSpace))
   > Set FACIAL_C_Jawline_Translation(FACIAL_C_Jawline_Translation:Vector=@n14) -> (Output_Get:Vector)
   > Set FACIAL_C_Jawline_Rotation(FACIAL_C_Jawline_Rotation:Rotator=@n14.Rotation) -> (Output_Get:Rotator)
   n15 = BreakTransform(InTransform=GetSocketTransform(self=@n9, InSocketName=FACIAL_L_Jawline, TransformSpace=RTS_ParentBoneSpace))
   > Set FACIAL_L_Jawline_Translation(FACIAL_L_Jawline_Translation:Vector=@n15) -> (Output_Get:Vector)
   > Set FACIAL_L_Jawline_Rotation(FACIAL_L_Jawline_Rotation:Rotator=@n15.Rotation) -> (Output_Get:Rotator)
   n16 = BreakTransform(InTransform=GetSocketTransform(self=@n9, InSocketName=FACIAL_R_Jawline, TransformSpace=RTS_ParentBoneSpace))
   > Set FACIAL_R_Jawline_Translation(FACIAL_R_Jawline_Translation:Vector=@n16) -> (Output_Get:Vector)
   > Set FACIAL_R_Jawline_Rotation(FACIAL_R_Jawline_Rotation:Rotator=@n16.Rotation) -> (Output_Get:Rotator)
   n17 = BreakTransform(InTransform=GetSocketTransform(self=@n9, InSocketName=FACIAL_L_Ear, TransformSpace=RTS_ParentBoneSpace))
   > Set FACIAL_L_Ear_Translation(FACIAL_L_Ear_Translation:Vector=@n17) -> (Output_Get:Vector)
   > Set FACIAL_L_Ear_Rotation(FACIAL_L_Ear_Rotation:Rotator=@n17.Rotation) -> (Output_Get:Rotator)
   n18 = BreakTransform(InTransform=GetSocketTransform(self=@n9, InSocketName=FACIAL_R_Ear, TransformSpace=RTS_ParentBoneSpace))
   > Set FACIAL_R_Ear_Translation(FACIAL_R_Ear_Translation:Vector=@n18) -> (Output_Get:Vector)
   > Set FACIAL_R_Ear_Rotation(FACIAL_R_Ear_Rotation:Rotator=@n18.Rotation) -> (Output_Get:Rotator)
   n19 = BreakTransform(InTransform=GetSocketTransform(self=@n9, InSocketName=FACIAL_L_CheekLower, TransformSpace=RTS_ParentBoneSpace))
   > Set FACIAL_L_CheekLower_Translation(FACIAL_L_CheekLower_Translation:Vector=@n19) -> (Output_Get:Vector)
   > Set FACIAL_L_CheekLower_Rotation(FACIAL_L_CheekLower_Rotation:Rotator=@n19.Rotation) -> (Output_Get:Rotator)
   n20 = BreakTransform(InTransform=GetSocketTransform(self=@n9, InSocketName=FACIAL_R_CheekLower, TransformSpace=RTS_ParentBoneSpace))
   > Set FACIAL_R_CheekLower_Translation(FACIAL_R_CheekLower_Translation:Vector=@n20) -> (Output_Get:Vector)
   > Set FACIAL_R_CheekLower_Rotation(FACIAL_R_CheekLower_Rotation:Rotator=@n20.Rotation) -> (Output_Get:Rotator)
   n21 = BreakTransform(InTransform=GetSocketTransform(self=@n9, InSocketName=FACIAL_L_ChinSide, TransformSpace=RTS_ParentBoneSpace))
   > Set FACIAL_L_ChinSide_Translation(FACIAL_L_ChinSide_Translation:Vector=@n21) -> (Output_Get:Vector)
   > Set FACIAL_L_ChinSide_Rotation(FACIAL_L_ChinSide_Rotation:Rotator=@n21.Rotation) -> (Output_Get:Rotator)
   n22 = BreakTransform(InTransform=GetSocketTransform(self=@n9, InSocketName=FACIAL_R_ChinSide, TransformSpace=RTS_ParentBoneSpace))
   > Set FACIAL_R_ChinSide_Translation(FACIAL_R_ChinSide_Translation:Vector=@n22) -> (Output_Get:Vector)
   > Set FACIAL_R_ChinSide_Rotation(FACIAL_R_ChinSide_Rotation:Rotator=@n22.Rotation) -> (Output_Get:Rotator)
   n23 = Get bIsSkyDiving()
   n24 = Get bIsPlayingEmote()
   Set isInIdlePoseFrontEnd(execute=@n2.then_4, isInIdlePoseFrontEnd:bool=AND Boolean(A=Get bIsInFrontEnd(), B=NOT Boolean(A=@n24), C=NOT Boolean(A=@n23))) -> (Output_Get:bool)
   n25 = Get bIsMoving2D()
   Set isCrouchingStill(execute=@n2.then_2, isCrouchingStill:bool=AND Boolean!(A=Get bIsCrouching(), B=NOT Boolean(A=@n25)) -> (ReturnValue)) -> (Output_Get:bool)
   n26 = Gravity Override State Function DEPRECATED(
    execute=@n2.then_0
    CharacterPartAnimInstance=Self-Reference()
    GameTimeInSeconds=GetGameTimeInSeconds()
    GravityOverrideParams=Get Gravity Override Params()
   )
   > Set GravityOverride(GravityOverride=@n26)
   n27 = Event OnCharacterCustomizationLoaded()
   > Set HeadMesh(HeadMesh:Skeletal Mesh Component Budgeted Object Reference=Get Array Item with EFort Custom Part Type(Array=@n27.CustomMeshParts, PartType=Head)) -> (Output_Get:Skeletal Mesh Component Budgeted Object Reference)
   Event OnParachuteClosed()
   > Set isUsingHangGliding(isUsingHangGliding:bool=false) -> (Output_Get:bool)
   Event OnParachuteOpened()
   > Set isUsingHangGliding(isUsingHangGliding=false)
   > Switch on EGliderType(Selection=Get PlayerGliderType() -> (PlayerGliderType=HangGlider))
   > Set isUsingHangGliding(isUsingHangGliding=true)
   n28 = Event BlueprintProcessAnimInputEvent()
   > n29 = Branch(Condition=IsNamed(self=@n28.AnimInputEvent, InEventName=aie.Player.Zipline.Entered))
   > Set onZipline(onZipline:bool=1) -> (Output_Get:bool)
   Branch(execute=@n29.else, Condition=IsNamed(self=@n28.AnimInputEvent, InEventName=aie.Player.Zipline.Exited))
   > Set onZipline(onZipline=false)
  ]}
]
variables [
 Delta Time X {type=number}
 FACIAL_C_Chin_Rotation {type=Rotator}
 FACIAL_C_Chin_Translation {type=Vector}
 FACIAL_C_Jaw_Rotation {type=Rotator}
 FACIAL_C_Jaw_Tranform {type=Vector}
 FACIAL_C_Jawline_Rotation {type=Rotator}
 FACIAL_C_Jawline_Translation {type=Vector}
 FACIAL_L_CheekLower_Rotation {type=Rotator}
 FACIAL_L_CheekLower_Translation {type=Vector}
 FACIAL_L_ChinSide_Rotation {type=Rotator}
 FACIAL_L_ChinSide_Translation {type=Vector}
 FACIAL_L_Ear_Rotation {type=Rotator}
 FACIAL_L_Ear_Translation {type=Vector}
 FACIAL_L_Jawline_Rotation {type=Rotator}
 FACIAL_L_Jawline_Translation {type=Vector}
 FACIAL_L_UnderChin_Rotation {type=Rotator}
 FACIAL_L_UnderChin_Translation {type=Vector}
 FACIAL_R_CheekLower_Rotation {type=Rotator}
 FACIAL_R_CheekLower_Translation {type=Vector}
 FACIAL_R_ChinSide_Rotation {type=Rotator}
 FACIAL_R_ChinSide_Translation {type=Vector}
 FACIAL_R_Ear_Rotation {type=Rotator}
 FACIAL_R_Ear_Translation {type=Vector}
 FACIAL_R_Jawline_Rotation {type=Rotator}
 FACIAL_R_Jawline_Translation {type=Vector}
 FACIAL_R_UnderChin_Rotation {type=Rotator}
 FACIAL_R_UnderChin_Translation {type=Vector}
 Gravity Override Params {type=GravityOverrideParamsStruct}
 GravityOverride {type=Vector}
 HeadMesh {type=SkeletalMeshComponentBudgeted}
 isCrouchingStill {type=boolean}
 isInIdlePoseFrontEnd {type=boolean}
 isStandingStill {type=boolean}
 isTargeting float {type=number}
 isUsingFacialBones {type=boolean}
 isUsingHangGliding {type=boolean}
 onZipline {type=boolean}
]
