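# Physics asset for the center front hair strand of the F_MED_RoseDepth_Bloom female character's face accessory mesh.
# A base capsule links to the head with free swing and twist, followed by a chain of four dynamic hair capsules.
# The chain has zero twist limits and narrow swing limits that generally loosen away from the root.
# The head body combines skull and nose capsules, giving the face a collision shape.
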PhysicsAsset /BRCosmetics/Characters/Player/Female/Medium/Bodies/F_MED_RoseDepth_Bloom/Meshes/Parts/F_MED_RoseDepth_Bloom_FaceAcc_Physics_frontHair
bodies {
 base_frontHair_c_1 [{shapeType=Capsule, length=2, radius=1, shapeName=base_frontHair_c_1_capsule}]
 dyn_frontHair_c_1 [{shapeType=Capsule, length=4, radius=2, shapeName=dyn_frontHair_c_1_capsule, center=<2.4796 0 0>, rotation=<rot 0 0 90>}]
 dyn_frontHair_c_2 [{shapeType=Capsule, length=4, radius=2, shapeName=dyn_frontHair_c_1_capsule, center=<2.4796 0 0>, rotation=<rot 0 0 90>}]
 dyn_frontHair_c_3 [{
   shapeType=Capsule
   length=2.4424
   radius=1.3547
   shapeName=dyn_frontHair_c_3_capsule
   center=<2.6512 -0.0052 0.4988>
   rotation=<rot -0.0792 -98.529 -93.6868>
  }]
 dyn_frontHair_c_4 [{
   shapeType=Capsule
   length=4.3156
   radius=0.6806
   shapeName=dyn_frontHair_c_4_capsule
   center=<2.4958 0.0783 0.3329>
   rotation=<rot -0.0017 -89.9994 -95.3455>
  }]
 head [
  {shapeType=Capsule, length=6.3093, radius=6.1999, shapeName=head_capsule, center=<1.4229 3.1204 0>, rotation=<rot 90 0 0>}
  {shapeType=Capsule, length=3.6671, radius=1.0789, shapeName=nose_capsule, center=<1.5676 8.701 0>, rotation=<rot 90 180 -155.0562>}
 ]
}
constraints [
 {bone1Name=dyn_frontHair_c_3, bone2Name=dyn_frontHair_c_2, swing1LimitDegrees=10, swing2LimitDegrees=20, twistLimitDegrees=0}
 {bone1Name=dyn_frontHair_c_4, bone2Name=dyn_frontHair_c_3, swing1LimitDegrees=10, swing2LimitDegrees=15, twistLimitDegrees=0}
 {bone1Name=dyn_frontHair_c_2, bone2Name=dyn_frontHair_c_1, swing1LimitDegrees=5, swing2LimitDegrees=10, twistLimitDegrees=0}
 {
  bone1Name=base_frontHair_c_1
  bone2Name=head
  swing1LimitDegrees=45
  swing1Motion=Free
  swing2LimitDegrees=45
  swing2Motion=Free
  twistLimitDegrees=45
  twistMotion=Free
 }
 {bone1Name=dyn_frontHair_c_1, bone2Name=base_frontHair_c_1, swing1LimitDegrees=0, swing2LimitDegrees=5, twistLimitDegrees=0}
]
