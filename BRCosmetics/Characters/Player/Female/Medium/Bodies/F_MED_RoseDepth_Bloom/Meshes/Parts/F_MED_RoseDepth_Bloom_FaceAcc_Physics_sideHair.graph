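# Physics asset for the side hair of the female medium RoseDepth Bloom face accessory, with capsule bodies for six strands in front, front2 and back pairs.
# Each strand chains two or three dyn bones to a small base body, using narrow swing limits and zero twist.
# Base bodies attach to the head capsule with free angular motion.
# An unconstrained capsule covers only the right upper arm, and left and right chains differ in segment count and shape.
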PhysicsAsset /BRCosmetics/Characters/Player/Female/Medium/Bodies/F_MED_RoseDepth_Bloom/Meshes/Parts/F_MED_RoseDepth_Bloom_FaceAcc_Physics_sideHair
bodies {
 base_hairSideBack_l_1 [{shapeType=Capsule, length=1, radius=0.5, shapeName=base_hairSideBack_l_1_capsule}]
 base_hairSideBack_r_1 [{shapeType=Capsule, length=1, radius=0.5, shapeName=base_hairSideBack_r_1_capsule}]
 base_hairSideFront2_l_1 [{shapeType=Capsule, length=1, radius=0.5, shapeName=base_hairSideFront2_l_1_capsule}]
 base_hairSideFront2_r_1 [{shapeType=Capsule, length=1, radius=0.5, shapeName=base_hairSideFront2_r_1_capsule}]
 base_hairSideFront_l_1 [{shapeType=Capsule, length=1, radius=0.5, shapeName=base_hairSideFront_l_1_capsule}]
 base_hairSideFront_r_1 [{shapeType=Capsule, length=1, radius=0.5, shapeName=base_hairSideFront_r_1_capsule}]
 dyn_hairSideBack_l_1 [{
   shapeType=Capsule
   length=2.116
   radius=2.6018
   shapeName=dyn_hairSideBack_l_1_capsule
   center=<4.1304 0.1209 0.4007>
   rotation=<rot -16.8279 -92.7626 -82.7625>
  }]
 dyn_hairSideBack_l_2 [{
   shapeType=Capsule
   length=4.1546
   radius=3.054
   shapeName=dyn_hairSideBack_l_2_capsule
   center=<4.0337 0.1236 -0.4955>
   rotation=<rot -7.2532 -101.978 -84.8967>
  }]
 dyn_hairSideBack_r_1 [{
   shapeType=Capsule
   length=3.4196
   radius=2.6043
   shapeName=dyn_hairSideFront_r_1_capsule
   center=<3.4332 0 0.751>
   rotation=<rot -86.1488 -179.9997 179.9997>
  }]
 dyn_hairSideBack_r_2 [{
   shapeType=Capsule
   length=2.1101
   radius=2.9952
   shapeName=dyn_hairSideFront_r_2_capsule
   center=<3.689 0 0.1289>
   rotation=<rot -86.6394 -179.9998 179.9998>
  }]
 dyn_hairSideBack_r_3 [{
   shapeType=Capsule
   length=3.0229
   radius=3.0279
   shapeName=dyn_hairSideFront_r_3_capsule
   center=<4.3717 0 0.1168>
   rotation=<rot -86.7887 -179.9999 179.9998>
  }]
 dyn_hairSideFront2_l_1 [{
   shapeType=Capsule
   length=4.0508
   radius=2.3461
   shapeName=dyn_hairSideFront2_l_1_capsule
   center=<2.9671 -1.2266 1.0984>
   rotation=<rot 0.0537 93.0819 -95.5883>
  }]
 dyn_hairSideFront2_l_2 [{
   shapeType=Capsule
   length=4.6275
   radius=1.7917
   shapeName=dyn_hairSideFront2_l_2_capsule
   center=<3.7857 -0.4674 0.7163>
   rotation=<rot 2.235 -83.6136 -91.8768>
  }]
 dyn_hairSideFront2_l_3 [{
   shapeType=Capsule
   length=2.2501
   radius=1.0355
   shapeName=dyn_hairSideFront2_l_3_capsule
   center=<3.3341 -0.0939 0.0507>
   rotation=<rot 0.0103 -88.1728 -89.7676>
  }]
 dyn_hairSideFront2_r_1 [{
   shapeType=Capsule
   length=3.8326
   radius=1.5838
   shapeName=dyn_hairSideFront2_r_1_capsule
   center=<2.9869 0 0.4403>
   rotation=<rot -86.0211 -179.9999 179.9999>
  }]
 dyn_hairSideFront2_r_2 [{
   shapeType=Capsule
   length=4.3566
   radius=1.8379
   shapeName=dyn_hairSideFront2_r_2_capsule
   center=<4.255 0 0.3616>
   rotation=<rot -86.8878 0 0>
  }]
 dyn_hairSideFront2_r_3 [{
   shapeType=Capsule
   length=3.1122
   radius=1.188
   shapeName=dyn_hairSideFront2_r_3_capsule
   center=<2.9953 0.4091 0>
   rotation=<rot -90 8.4016 -14.1715>
  }]
 dyn_hairSideFront_l_1 [{
   shapeType=Capsule
   length=2.5139
   radius=2.3191
   shapeName=dyn_hairSideFront_l_1_capsule
   center=<4.1089 0.0767 1.5961>
   rotation=<rot 0.422 -92.5949 -88.371>
  }]
 dyn_hairSideFront_l_2 [{
   shapeType=Capsule
   length=3.281
   radius=2.5813
   shapeName=dyn_hairSideFront_l_2_capsule
   center=<4.6009 -0.7343 0.5664>
   rotation=<rot -3.6109 -96.9769 -75.6961>
  }]
 dyn_hairSideFront_r_1 [{
   shapeType=Capsule
   length=3.4196
   radius=2.6043
   shapeName=dyn_hairSideFront_r_1_capsule
   center=<3.4332 0 0.751>
   rotation=<rot -86.1488 -179.9997 179.9997>
  }]
 dyn_hairSideFront_r_2 [{
   shapeType=Capsule
   length=2.1101
   radius=2.9952
   shapeName=dyn_hairSideFront_r_2_capsule
   center=<3.689 0 0.1289>
   rotation=<rot -86.6394 -179.9998 179.9998>
  }]
 dyn_hairSideFront_r_3 [{
   shapeType=Capsule
   length=6.1305
   radius=2.4901
   shapeName=dyn_hairSideFront_r_3_capsule
   center=<5.7597 0 0.0389>
   rotation=<rot -83.7181 -0.0003 0.0002>
  }]
 head [{shapeType=Capsule, length=6.4672, radius=6.042, shapeName=head_capsule, center=<1.4229 3.1204 0>, rotation=<rot 90 0 0>}]
 upperarm_r [{shapeType=Capsule, length=20.4752, radius=4.8162, shapeName=upperarm_r_capsule, center=<-13.2773 0 1.2202>, rotation=<rot 0 -90 90>}]
}
constraints [
 {bone1Name=dyn_hairSideBack_l_1, bone2Name=base_hairSideBack_l_1, swing1LimitDegrees=8, swing2LimitDegrees=8, twistLimitDegrees=0}
 {bone1Name=dyn_hairSideBack_l_2, bone2Name=dyn_hairSideBack_l_1, swing1LimitDegrees=7, swing2LimitDegrees=10, twistLimitDegrees=0}
 {bone1Name=dyn_hairSideFront2_l_1, bone2Name=base_hairSideFront2_l_1, swing1LimitDegrees=10, swing2LimitDegrees=15, twistLimitDegrees=0}
 {bone1Name=dyn_hairSideFront2_l_2, bone2Name=dyn_hairSideFront2_l_1, swing1LimitDegrees=15, swing2LimitDegrees=15, twistLimitDegrees=0}
 {bone1Name=dyn_hairSideFront2_l_3, bone2Name=dyn_hairSideFront2_l_2, swing1LimitDegrees=10, swing2LimitDegrees=10, twistLimitDegrees=0}
 {bone1Name=dyn_hairSideFront_l_2, bone2Name=dyn_hairSideFront_l_1, swing1LimitDegrees=7, swing2LimitDegrees=10, twistLimitDegrees=0}
 {bone1Name=dyn_hairSideFront_l_1, bone2Name=base_hairSideFront_l_1, swing1LimitDegrees=8, swing2LimitDegrees=8, twistLimitDegrees=0}
 {
  bone1Name=base_hairSideFront_l_1
  bone2Name=head
  swing1LimitDegrees=45
  swing1Motion=Free
  swing2LimitDegrees=45
  swing2Motion=Free
  twistLimitDegrees=45
  twistMotion=Free
 }
 {
  bone1Name=base_hairSideFront2_l_1
  bone2Name=head
  swing1LimitDegrees=45
  swing1Motion=Free
  swing2LimitDegrees=45
  swing2Motion=Free
  twistLimitDegrees=45
  twistMotion=Free
 }
 {
  bone1Name=base_hairSideBack_l_1
  bone2Name=head
  swing1LimitDegrees=45
  swing1Motion=Free
  swing2LimitDegrees=45
  swing2Motion=Free
  twistLimitDegrees=45
  twistMotion=Free
 }
 {bone1Name=dyn_hairSideBack_r_2, bone2Name=dyn_hairSideBack_r_1, swing1LimitDegrees=7, swing2LimitDegrees=10, twistLimitDegrees=0}
 {bone1Name=dyn_hairSideBack_r_3, bone2Name=dyn_hairSideBack_r_2, swing1LimitDegrees=10, swing2LimitDegrees=15, twistLimitDegrees=0}
 {bone1Name=dyn_hairSideBack_r_1, bone2Name=base_hairSideBack_r_1, swing1LimitDegrees=8, swing2LimitDegrees=8, twistLimitDegrees=0}
 {
  bone1Name=base_hairSideBack_r_1
  bone2Name=head
  swing1LimitDegrees=45
  swing1Motion=Free
  swing2LimitDegrees=45
  swing2Motion=Free
  twistLimitDegrees=45
  twistMotion=Free
 }
 {bone1Name=dyn_hairSideFront2_r_3, bone2Name=dyn_hairSideFront2_r_2, swing1LimitDegrees=10, swing2LimitDegrees=10, twistLimitDegrees=0}
 {bone1Name=dyn_hairSideFront2_r_2, bone2Name=dyn_hairSideFront2_r_1, swing1LimitDegrees=15, swing2LimitDegrees=15, twistLimitDegrees=0}
 {bone1Name=dyn_hairSideFront2_r_1, bone2Name=base_hairSideFront2_r_1, swing1LimitDegrees=10, swing2LimitDegrees=15, twistLimitDegrees=0}
 {
  bone1Name=base_hairSideFront2_r_1
  bone2Name=head
  swing1LimitDegrees=45
  swing1Motion=Free
  swing2LimitDegrees=45
  swing2Motion=Free
  twistLimitDegrees=45
  twistMotion=Free
 }
 {bone1Name=dyn_hairSideFront_r_1, bone2Name=base_hairSideFront_r_1, swing1LimitDegrees=8, swing2LimitDegrees=8, twistLimitDegrees=0}
 {bone1Name=dyn_hairSideFront_r_3, bone2Name=dyn_hairSideFront_r_2, swing1LimitDegrees=10, swing2LimitDegrees=15, twistLimitDegrees=0}
 {bone1Name=dyn_hairSideFront_r_2, bone2Name=dyn_hairSideFront_r_1, swing1LimitDegrees=7, swing2LimitDegrees=10, twistLimitDegrees=0}
 {
  bone1Name=base_hairSideFront_r_1
  bone2Name=head
  swing1LimitDegrees=45
  swing1Motion=Free
  swing2LimitDegrees=45
  swing2Motion=Free
  twistLimitDegrees=45
  twistMotion=Free
 }
]
